# Copyright Epic Games, Inc. All Rights Reserved.
#################################################
# Generated Digest of Verse API
# DO NOT modify this manually!
# Generated from build: ++Fortnite+Release-37.00-CL-44765404
#################################################


using {/Verse.org/Native}
using {/Verse.org/Simulation}
using {/Verse.org/Assets}
using {/Verse.org/SceneGraph}
# Module import path: /playkigai.dig@fortnite.com/Minigame_MusicalChairs/Meshes
Meshes := module:
    SM_PlayerRegister<scoped {/playkigai.dig@fortnite.com/Minigame_MusicalChairs}> := class<final><scoped {/playkigai.dig@fortnite.com/Minigame_MusicalChairs}>(mesh_component):
        @editable
        var WorldGridMaterial<public>:material = external {}

    SM_PlayerRegister_asset<scoped {/playkigai.dig@fortnite.com/Minigame_MusicalChairs}>:mesh = external {}

# Module import path: /playkigai.dig@fortnite.com/Minigame_MusicalChairs/Prefabs
Prefabs := module:
    @import_as("/Minigame_MusicalChairs/Prefabs/game_manager.game_manager_C")
    game_manager<scoped {/playkigai.dig@fortnite.com/Minigame_MusicalChairs}> := class<final>(entity):


    game_manager_asset<scoped {/playkigai.dig@fortnite.com/Minigame_MusicalChairs}>:entity_prefab = external {}

    @import_as("/Minigame_MusicalChairs/Prefabs/player_register.player_register_C")
    player_register<scoped {/playkigai.dig@fortnite.com/Minigame_MusicalChairs}> := class<final>(entity):


    player_register_asset<scoped {/playkigai.dig@fortnite.com/Minigame_MusicalChairs}>:entity_prefab = external {}

    @import_as("/Minigame_MusicalChairs/Prefabs/round_manager.round_manager_C")
    round_manager<scoped {/playkigai.dig@fortnite.com/Minigame_MusicalChairs}> := class<final>(entity):


    round_manager_asset<scoped {/playkigai.dig@fortnite.com/Minigame_MusicalChairs}>:entity_prefab = external {}