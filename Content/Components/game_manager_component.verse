<#
@UUID: 26c38fdb-53cd-4eda-8651-8f8848b31bf7
@Task: Game Manager
@Description: The Game Manager handles overall gamestate and coordinates each working piece of the minigame itself. It signals when a new game starts, handles starting a round, ending a round, player's in queue, running the game itself, and determining a winner (and losers).
@Tags: Game Manager, Game State, Winners, Losers, Winning, Round Management, Start Game, Manage Game, Game Settings
@Export: true
#>
<#
Entity Prefab setup for `game_manager`:
    - game_manager
        - transform_component
        - mesh_component
            NOTE:
                mesh_component is not necessary for the round_manager, but it can be beneficial to include a mesh in order to more easily find it in the world.
        - game_manager_component
#>
using { /Fortnite.com/Playspaces }
using { /Verse.org/SceneGraph }
using { /Verse.org/Simulation }

using { Prefabs }


# Component that manages the game.
game_manager_component<public> := class<final_super>(component):
    @editable
    MinPlayerCount: int = 2
    @editable
    MaxPlayerCount: int = 5

    @editable
    RoundManager: round_manager = round_manager{}

    GameStartedEvent: event() = event(){}

    var CurrentPlayers: []agent = array{}


    # === Entity Functions ===

    OnBeginSimulation<override>():void=
        # Find and initialize the player_register_component and RoundManager.
        if:
            SimEntity := Entity.GetSimulationEntity[]
            PlayerRegCompGenerator := SimEntity.FindDescendantComponents(player_register_component)
            PlayerRegCompArr := for (PlayerRegCompItem : PlayerRegCompGenerator) do PlayerRegCompItem
            PlayerRegComp := PlayerRegCompArr[0]
        then:
            PlayerRegComp.SetGameManagerRef(Self)
            PlayerRegComp.SetMinPlayers(MinPlayerCount)
            PlayerRegComp.SetMaxPlayers(MaxPlayerCount)
        else:
            Print(PlayerRegisterErrorText)

        if (Playspace := Entity.GetPlayspaceForEntity[]):
            Playspace.PlayerRemovedEvent().Subscribe(OnPlayerRemoved)


    OnSimulate<override>()<suspends>:void=
        loop:
            GameStartedEvent.Await()

            RunGame()


    # === Game Manager Component Functions ===

    # Handle a player being removed.
    OnPlayerRemoved(InAgent:agent):void=
        if (CurrentPlayers.RemoveFirstElement[InAgent]) {}


    # Core game loop.
    RunGame()<suspends>:void=
        RoundManagerComp :=
            if (RMComp := RoundManager.GetRoundManagerComponent[]):
                RMComp
            else:
                Print(RoundManagerCompErrorText)
                return

        loop:
            spawn { RoundManager.RunRound(CurrentPlayers.Length) }

            # RoundEnded event provides an array of players that are advancing to the next round.
            set CurrentPlayers = RoundManagerComp.RoundEndedEvent.Await()

            if (CurrentPlayers.Length = 0):
                DeclareTie()
                return

            if (CurrentPlayers.Length = 1, Winner := CurrentPlayers[0]):
                DeclareWinner(Winner)
                return


    # Declare the winner of the game.
    DeclareWinner(Winner:agent):void=
        Print("{Winner} is the winner!")
        return


    # Declare the game as a tie.
    DeclareTie():void=
        Print("The game has ended in a tie!")
        return


# === Extension Functions ===

# Initialize players for the game.
(GameManagerComp:game_manager_component).InitializePlayers(ReadyAgentsMap:[agent]logic):void=
    for:
        Agent -> IsReady : ReadyAgentsMap
        IsReady?
    do:
        set GameManagerComp.CurrentPlayers += array{Agent}


# Start the game.
(GameManagerComp:game_manager_component).StartGame():void=
    GameManagerComp.GameStartedEvent.Signal()


# === Agent Name Printing ===

AgentNameText<localizes>(Agent:agent):message= "{Agent}"
ToString(Agent:agent):string= Localize(AgentNameText(Agent))