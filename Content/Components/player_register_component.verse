<#
@UUID: 3d94631b-3789-42cc-aa97-60a4a6953be9
@Task: Player Register
@Description: The Player Register allows players to enter and exit queue. Players that are in queue whenever a new game is started are included as active players for the duration of a game. If a player is not in queue, but interacts with the Player Register entity, they will be added to queue. If a player interacts with the Player Register entity while currently in queue will cause the interacting players to leave queue (and thus will not be included as an active player whenever a new game starts).
@Tags: Player Register, Queue, Interactable Object
@Export: true
#>
<#
Entity Prefab setup for `player_register`:
    - player_register
        - transform_component
        - mesh_component
            NOTE:
                mesh_component *must* have an associated mesh with collision enabled in order for it to be interactable by players.
        - player_register_component
#>
using { /Fortnite.com/Devices }
using { /Verse.org/SceneGraph }
using { /Verse.org/Simulation }


# Interactable component that handles registering and unregistering players for a game.
player_register_component<public> := class<final_super>(interactable_component):
    @editable
    StartButton: button_device = button_device{}

    var MaybeGameManagerRef: ?game_manager_component = false
    var MinPlayers: int = 0
    var MaxPlayers: int = 0
    var NumCurrentPlayers: int = 0
    var QueuedPlayers: [agent]logic = map{}


    # === Entity Functions ===

    OnBeginSimulation<override>():void=
        SucceededEvent.Subscribe(OnInteractedWith)
        StartButton.InteractedWithEvent.Subscribe(OnStartGameButtonPress)


    # === Player Register Component Functions ===

    # Handle the InteractedWith event for the player register.
    OnInteractedWith(InAgent:agent):void=
        # Unregister player.
        if (QueuedPlayers[InAgent]?, set QueuedPlayers[InAgent] = false):
            Print("{InAgent} left queue.")
            set NumCurrentPlayers -= 1
            return

        # Register player.
        if (NumCurrentPlayers < MaxPlayers, set QueuedPlayers[InAgent] = true):
            Print("{InAgent} joined queue.")
            set NumCurrentPlayers += 1
            return

    
    # Handle the InteractedWith event for the StartButton.
    OnStartGameButtonPress(InAgent:agent):void=
        if (NumCurrentPlayers < MinPlayers):
            Print(NotEnoughPlayersErrorText)
            return

        if (GMRef := MaybeGameManagerRef?):
            GMRef.InitializePlayers(QueuedPlayers)
            GMRef.StartGame()

    
    # === Messages ===

    NotEnoughPlayersErrorText<localizes>:message = "Not enough players to start game."


# === Extension Functions ===

# Set the MinPlayers variable.
(PRC:player_register_component).SetMinPlayers(InMinPlayers:int):void=
    set PRC.MinPlayers = InMinPlayers


# Set the MaxPlayers variable.
(PRC:player_register_component).SetMaxPlayers(InMaxPlayers:int):void=
    set PRC.MaxPlayers = InMaxPlayers


# Set the GameManagerRef variable.
(PRC:player_register_component).SetGameManagerRef(InGMRef:game_manager_component):void=
    set PRC.MaybeGameManagerRef = option{InGMRef}


# === Messages ===

PlayerRegisterErrorText<localizes>:message = "Unable to find player_register component."