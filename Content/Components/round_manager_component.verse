<#
@UUID: f1f9e06d-8811-4c17-8ec1-b040ec1ef954
@Task: Round Manager
@Description: The Round Manager handles round-specific logic such as moving, rotating, and positioning chairs as well as controlling the duration of a round, the duration of music, and whether or not chairs are able to be interacted with. While the `game_manager` handles overarching logic and sequencing, the `round_manager` handles the operations of the playable portion of the minigame.
@Tags: Round Manager, Round Logic, Devices, Transforms
@Export: true
#>
<#
Entity Prefab setup for round_manager:
    - round_manager
        - transform_component
        - mesh_component
            NOTE:
                mesh_component is not necessary for the round_manager, but it can be beneficial to include a mesh in order to more easily find it in the world.
        - round_manager_component
#>
using { /Fortnite.com/Devices }
using { /Verse.org }
using { /Verse.org/Random }
using { /Verse.org/SceneGraph }
using { /Verse.org/Simulation }
using { /Verse.org/SpatialMath }

using { Prefabs }

UESM := import (/UnrealEngine.com/Temporary/SpatialMath)

# Component that runs a round of musical chairs. 
round_manager_component<public> := class<final_super>(component):
    @editable { ToolTip := MinMusicTimerToolTipText, Categories := array{TimerCategoryText} }
    MinMusicTimer: float = 1.5
    @editable { ToolTip := MinMusicTimerToolTipText, Categories := array{TimerCategoryText} }
    MaxMusicTimer: float = 6.0

    @editable { ToolTip := MinMusicTimerToolTipText, Categories := array{TimerCategoryText} }
    MinSeatingTimer: float = 1.0
    @editable { ToolTip := MinMusicTimerToolTipText, Categories := array{TimerCategoryText} }
    MaxSeatingTimer: float = 1.5

    @editable { ToolTip := MinMusicTimerToolTipText, Categories := array{ChairPlacementCategoryText} }
    ChairOrigin: vector3 = vector3{}
    @editable { ToolTip := MinMusicTimerToolTipText, Categories := array{ChairPlacementCategoryText} }
    ChairSpawnRadius: float = 200.0

    @editable { ToolTip := MinMusicTimerToolTipText, Categories := array{DevicesCategoryText} }
    Chairs: []chair_device = array{}
    @editable { ToolTip := MinMusicTimerToolTipText, Categories := array{DevicesCategoryText} }
    AudioPlayer: audio_player_device = audio_player_device{}

    MusicTimerFinishedEvent:   event() = event(){}
    SeatingTimerFinishedEvent: event() = event(){}
    RoundEndedEvent:           event([]agent) = event([]agent){}


    # Used as a sleep timer, and to decrement a player's action lock timer.
    CountdownSleepTimer: float = 0.1
    TwoPI: float = 6.28

    var RoundMusicTime:   float = 5.0
    var RoundSeatingTime: float = 2.5
    var MusicTimer:       float = 5.0
    var SeatingTimer:     float = 2.5
    var NumPlayers:       int = 0


    # === Game Round Component Functions ===

    # Initialize the game round.
    InitializeRound():void=
        set RoundMusicTime   = GetRandomFloat(MinMusicTimer, MaxMusicTimer)
        set RoundSeatingTime = GetRandomFloat(MinSeatingTimer, MaxSeatingTimer)


    # Run the round.
    RunRoundInternal()<suspends>:void=
        set MusicTimer   = RoundMusicTime
        set SeatingTimer = RoundMusicTime

        # Initialize the chairs.
        DisableChairs(NumPlayers)
        PositionChairs(NumPlayers)

        # Play music.
        AudioPlayer.Play()
        spawn { CountdownMusicTimer() }
        MusicTimerFinishedEvent.Await()
        AudioPlayer.Stop()

        # Enable chairs for sitting.
        EnableChairs(NumPlayers)
        spawn { CountdownSeatingTimer() }
        SeatingTimerFinishedEvent.Await()

        # Reset the chairs and record seated players.
        RoundEndedEvent.Signal(GetSeatedPlayers())
        DisableChairs(NumPlayers)


    # Async function to run music.
    CountdownMusicTimer()<suspends>:void=
        loop:
            if (MusicTimer <= 0.0):
                MusicTimerFinishedEvent.Signal()
                return

            Sleep(CountdownSleepTimer)
            set MusicTimer -= CountdownSleepTimer


    # Async function to run seat timer.
    CountdownSeatingTimer()<suspends>:void=
        loop:
            if (SeatingTimer <= 0.0):
                SeatingTimerFinishedEvent.Signal()
                return

            Sleep(CountdownSleepTimer)
            set SeatingTimer -= CountdownSleepTimer

    
    # Place chairs for the round.
    PositionChairs(InCount:int):void=
        for (Idx := 0..(InCount - 1), Chair := Chairs[Idx]):
            Angle := (TwoPI * Idx) / (1.0 * InCount)
            XPos  := Cos(Angle) * ChairSpawnRadius
            YPos  := Sin(Angle) * ChairSpawnRadius

            Position := vector3: 
                Left    := ChairOrigin.Left + XPos
                Up      := ChairOrigin.Up
                Forward := ChairOrigin.Forward + YPos

            NewRotation := MakeShortestRotationBetween(UESM.FromVector3(utility{}.GetChairForward(Chair)), ChairOrigin - Position)

            if (Chair.TeleportTo[UESM.FromVector3(Position), UESM.FromRotation(Entity.GetGlobalTransform().Rotation)]) {}


    # Enable InNumPlayers chairs.
    EnableChairs(InNumPlayers:int):void=
        for (Idx := 0..(InNumPlayers - 1), Chair := Chairs[Idx]):
            Chair.Enable()


    # Disable InNumPlayers chairs.
    DisableChairs(InNumPlayers:int):void=
        for (Idx := 0..(InNumPlayers - 1), Chair := Chairs[Idx]):
            # Disable includes ejecting the seated player.
            Chair.Disable()


    # Get the seated players.
    GetSeatedPlayers():[]agent=
        var SeatedPlayers: []agent = array{}
        for (Chair : Chairs):
            MaybeAgent := Chair.GetSeatedAgent()
            if (Agent := MaybeAgent?):
                set SeatedPlayers += array{Agent}

        SeatedPlayers


# === Extension Functions ===

# Return the round_manager_component of the round_manager entity.
(RoundManager:round_manager).RunRound(InNumPlayers:int)<suspends>:void=
    if (RMComp := RoundManager.GetRoundManagerComponent[]):
        set RMComp.NumPlayers = InNumPlayers
        RMComp.RunRoundInternal()


# Return the round_manager_component of the round_manager entity.
(RoundManager:round_manager).GetRoundManagerComponent()<decides><transacts>:round_manager_component=
    RMCompGen := RoundManager.FindDescendantComponents(round_manager_component)
    RMCompArr := for (RMComp : RMCompGen) do RMComp
    RMCompArr[0]


# === Messages ===
MinMusicTimerToolTipText<localizes>:message= "The minimum amount of time the music will play in a round."
MaxMusicTimerToolTipText<localizes>:message= "The maximum amount of time the music will play in a round."
MinSeatingTimerToolTipText<localizes>:message= "The minimum amount of time players will be allowed to find a seat."
MaxSeatingTimerToolTipText<localizes>:message= "The maximum amount of time players will be allowed to find a seat."
TimerCategoryText<localizes>:message= "Timer Options"
ChairPlacementCategoryText<localizes>:message= "Chair Placement Options"
DevicesCategoryText<localizes>:message= "Devices"
RoundManagerCompErrorText<localizes>:message = "Unable to find round_manager component."