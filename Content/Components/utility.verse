<#
@UUID: 5b725b09-9819-4a70-8370-12a04014296e
@Task: Utility Library
@Description: The Utility Library is a collection of reusable, general-purpose functions. Using a utility library allows for cleaner code by repetative logic.
@Tags: Library, Utility Functions
@Export: true
#>
using { /Fortnite.com/Devices }
using { /UnrealEngine.com/Temporary/SpatialMath }

utility<public> := class:
    GetChairForward<public>(InChair:chair_device):vector3=
        InChair.GetTransform().Rotation.GetLocalForward()