<#
@UUID: 6bf7a1d9-382b-4117-999e-8960328884b4
@Task: Musical Chairs
@Description: Musical Chairs is a round-based competitive multiplayer minigame where players move in a circular fashion around chairs that are placed radially around a point, facing outward. There are a limited number of chairs (totaling the number of players minus one). Music plays for the duration of a round, with the chairs rotating radially around their origin point (they are rotated to reduce the opportunity for cheating by preventing players from attempting to "camp" a chair). Music is played for the duration of a round, and after a semi-random amount of time, the music is stopped. Once the music has stopped, the chairs also stop rotating around their origin point. The chairs are then allowed to be interacted with (and subsequently sit in). Players attempt to claim a chair by sitting in it, with at least one person being without a chair at the end of the round. Any players not sitting in a chair when the round comes to an end is eliminated. A new round begins with a number of chairs equal to one less than the total number of active players. The last player remaining is declared the winner.
@Tags: Competitive, Multiplayer, Music, Chairs, Chair Device, Audio Player Device, Audio Player, Minigame, Round-based, 
@Export: true
#>